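# Control Rig for the F_MED_Shrapnel female body cosmetic that procedurally offsets jacket collar master bones as the shoulders move.
# Forward solve converts clavicle_l and upperarm_r rotations, relative to spine_05, to Euler angles and remaps them with clamping.
# Left clavicle drives pitch on the front, mid and back left collar masters, plus Y translation on mid and back.
# Right upper arm drives only the right mid collar's pitch, so the setup is asymmetric.
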ControlRigBlueprint /BRCosmetics/Characters/Player/Female/Medium/Bodies/F_MED_Shrapnel/Meshes/F_MED_Shrapnel_CtrlRig
graphs [
 RigVMFunctionLibrary {links [], nodes []}
 RigVMModel {nodes [
   RigUnit_BeginExecution = &RigUnit_BeginExecution()
   n1 = &RigVMFunction_MathQuaternionToEuler(RotationOrder=ZYX, Value=&RigUnit_GetRelativeBoneTransform(Bone=clavicle_l, Space=spine_05).Transform.Rotation)
   &RigUnit_ModifyTransforms(
    ItemToModify=((Item=(Type=Bone,Name="dyn_collar_l_mid_master"),Transform=(Rotation=(X=0,Y=0,Z=0,W=1),Translation=(X=0,Y=-5,Z=0),Scale3D=(X=1,Y=1,Z=1))))
    Weight=1
    ItemToModify.0.Transform.Translation.Y=&RigVMFunction_MathFloatRemap(SourceMaximum=50, SourceMinimum=80, TargetMaximum=-5, TargetMinimum=0, bClamp=true, Value=@n1.Result.Y).Result
    ItemToModify.0.Transform.Rotation=/Script/RigVM.RigVMFunction_MathQuaternionFromRotator(
     Rotator=(Pitch=-60,Yaw=0,Roll=0)
     Rotator.Pitch=&RigVMFunction_MathFloatRemap(SourceMaximum=50, SourceMinimum=80, TargetMaximum=-60, TargetMinimum=0, bClamp=true, Value=@n1.Result.Y).Result
    ) {is=struct RigVMUnitNode}.Result
    ExecutePin=@RigUnit_BeginExecution.ExecutePin
   )
   > &RigUnit_ModifyTransforms(
    ItemToModify=((Item=(Type=Bone,Name="dyn_collar_l_front_master"),Transform=(Rotation=(X=0,Y=0,Z=0,W=1),Translation=(X=0,Y=0,Z=0),Scale3D=(X=1,Y=1,Z=1))))
    Weight=1
    ItemToModify.0.Transform.Rotation=&RigVMFunction_MathQuaternionFromRotator(
     Rotator=(Pitch=-30,Yaw=0,Roll=0)
     Rotator.Pitch=&RigVMFunction_MathFloatRemap(SourceMaximum=50, SourceMinimum=80, TargetMaximum=-30, TargetMinimum=0, bClamp=true, Value=@n1.Result.Y).Result
    ).Result
   )
   > &RigUnit_ModifyTransforms(
    ItemToModify=((Item=(Type=Bone,Name="dyn_collar_l_back_master"),Transform=(Rotation=(X=0,Y=0,Z=0,W=1),Translation=(X=0,Y=3,Z=0),Scale3D=(X=1,Y=1,Z=1))))
    Weight=1
    ItemToModify.0.Transform.Translation.Y=&RigVMFunction_MathFloatRemap(SourceMaximum=50, SourceMinimum=80, TargetMaximum=3, TargetMinimum=0, bClamp=true, Value=@n1.Result.Y).Result
    ItemToModify.0.Transform.Rotation=&RigVMFunction_MathQuaternionFromRotator(
     Rotator=(Pitch=-15,Yaw=0,Roll=0)
     Rotator.Pitch=&RigVMFunction_MathFloatRemap(SourceMaximum=50, SourceMinimum=80, TargetMaximum=-15, TargetMinimum=0, bClamp=true, Value=@n1.Result.Y).Result
    ).Result
   )
   n2 = &RigVMFunction_MathQuaternionToEuler(RotationOrder=ZYX, Value=&RigUnit_GetRelativeBoneTransform(Bone=upperarm_r, Space=spine_05).Transform.Rotation)
   > &RigUnit_ModifyTransforms(
    ItemToModify=((Item=(Type=Bone,Name="dyn_collar_r_mid_master"),Transform=(Rotation=(X=0,Y=0,Z=0,W=1),Translation=(X=0,Y=0,Z=0),Scale3D=(X=1,Y=1,Z=1))))
    Weight=1
    ItemToModify.0.Transform.Rotation=&RigVMFunction_MathQuaternionFromRotator(
     Rotator=(Pitch=-60,Yaw=0,Roll=0)
     Rotator.Pitch=&RigVMFunction_MathFloatRemap(SourceMaximum=120, SourceMinimum=90, TargetMaximum=-20, TargetMinimum=0, bClamp=true, Value=@n2.Result.Z).Result
    ).Result
   )
  ]}
]
hierarchy [
 root {loc=<0 0 0>, rot=<0 0 0 1>}
 pelvis {parent=root, loc=<0 -0.5301 98.0571>, rot=<0.0127 0.707 -0.0127 -0.707>}
 spine_01 {parent=pelvis, loc=<6.1208 -0.9328 0>, rot=<0 0 0.034 -0.9994>}
 spine_02 {parent=spine_01, loc=<6.5996 -0.2095 0>, rot=<0 0 0.0238 -0.9997>}
 spine_03 {parent=spine_02, loc=<6.4914 -0.1237 0>, rot=<0 0 0.0149 -0.9999>}
 spine_04 {parent=spine_03, loc=<6.9598 0.0794 0>, rot=<0 0 -0.0132 -0.9999>}
 spine_05 {parent=spine_04, loc=<6.9987 0 0>, rot=<0 0 -0.0015 -1>}
 collar_base {parent=spine_05, loc=<3.7978 0 0>, rot=<0 0 0 1>}
 dyn_collar_r_front_master {parent=collar_base, loc=<4.4316 7.2541 4.47>, rot=<0.4398 0.0789 -0.1235 -0.886>}
 dyn_collar_r_front {parent=dyn_collar_r_front_master, loc=<0 0 0>, rot=<0 0 0 1>}
 dyn_collar_r_mid_master {parent=collar_base, loc=<6.1989 2.6846 8.3962>, rot=<0.1277 0.1155 -0.1319 -0.9762>}
 dyn_collar_r_mid {parent=dyn_collar_r_mid_master, loc=<0 0 0>, rot=<0 0 0 -1>}
 dyn_collar_r_back_master {parent=collar_base, loc=<8.1345 -2.5184 7.2284>, rot=<-0.3993 -0.0298 -0.0876 -0.9121>}
 dyn_collar_r_back {parent=dyn_collar_r_back_master, loc=<0 0 0>, rot=<0 0 0 -1>}
 dyn_collar_l_mid_master {parent=collar_base, loc=<6.1992 2.6847 -8.3962>, rot=<0.9562 0.1482 -0.0937 -0.2344>}
 dyn_collar_l_mid {parent=dyn_collar_l_mid_master, loc=<0 0 0>, rot=<0 0 0 1>}
 dyn_collar_l_back_master {parent=collar_base, loc=<8.1344 -2.5184 -7.2284>, rot=<0.9216 0.1121 -0.0031 0.3716>}
 dyn_collar_l_back {parent=dyn_collar_l_back_master, loc=<0 0 0>, rot=<0 0 0 -1>}
 dyn_collar_back_master {parent=collar_base, loc=<9.05 -5.535 -0.0066>, rot=<-0.7186 -0.093 -0.0855 -0.6839>}
 dyn_collar_back {parent=dyn_collar_back_master, loc=<0 0 0>, rot=<0 0 0 1>}
 dyn_collar_l_front_master {parent=collar_base, loc=<4.4314 7.254 -4.47>, rot=<0.9344 0.0507 0.0209 -0.3521>}
 dyn_collar_l_front {parent=dyn_collar_l_front_master, loc=<0 0 0>, rot=<0 0 0 -1>}
 dyn_collar_buckle_l {parent=dyn_collar_l_front, loc=<1.0015 -1.803 -0.1371>, rot=<-0.0119 0.0759 0.7215 -0.6881>}
 clavicle_l {parent=spine_05, loc=<5.4831 -0.1053 -3.2946>, rot=<0.0465 -0.7289 -0.009 -0.6829>}
 upperarm_l {parent=clavicle_l, loc=<9.8658 -1.2136 -0.4105>, rot=<0.0164 -0.3245 -0.0269 -0.9454>}
 lowerarm_l {parent=upperarm_l, loc=<26.465 -0.8345 0.259>, rot=<0.0036 0.0161 -0.2298 -0.9731>}
 hand_l {parent=lowerarm_l, loc=<22.7355 -0.3203 -0.6105>, rot=<-0.6784 -0.0223 0.0735 -0.7307>}
 index_metacarpal_l {parent=hand_l, loc=<1.6535 0.0183 -2.96>, rot=<0.0158 -0.0124 -0.0173 -0.9996>}
 index_01_l {parent=index_metacarpal_l, loc=<8.3008 0.0357 -0.0524>, rot=<-0.0044 0.0485 0.0924 -0.9945>}
 index_02_l {parent=index_01_l, loc=<3.3587 -0.3004 0.0452>, rot=<-0.002 0.0027 0.3225 -0.9466>}
 index_03_l {parent=index_02_l, loc=<2.5578 -0.123 -0.0252>, rot=<-0.002 0.0067 0.2558 -0.9667>}
 middle_metacarpal_l {parent=hand_l, loc=<2.1856 -0.1725 -0.5984>, rot=<0.006 0.0168 -0.0329 -0.9993>}
 middle_01_l {parent=middle_metacarpal_l, loc=<7.3275 -0.2447 0.0129>, rot=<-0.0208 0.0042 0.0858 -0.9961>}
 middle_02_l {parent=middle_01_l, loc=<4.0123 0.1401 0.1175>, rot=<-0.0218 0.0258 0.3512 -0.9357>}
 middle_03_l {parent=middle_02_l, loc=<2.6077 0.0534 0.1373>, rot=<0.0082 0.0188 0.1708 -0.9851>}
 pinky_metacarpal_l {parent=hand_l, loc=<1.7616 -0.9753 3.3814>, rot=<-0.0397 0.0814 -0.069 -0.9935>}
 pinky_01_l {parent=pinky_metacarpal_l, loc=<7.1709 -0.4831 -0.1094>, rot=<-0.0654 -0.0035 0.2242 -0.9723>}
 pinky_02_l {parent=pinky_01_l, loc=<2.8274 0.0893 -0.1043>, rot=<0.0037 0.0011 0.2854 -0.9584>}
 pinky_03_l {parent=pinky_02_l, loc=<2.2046 0.0307 -0.0078>, rot=<-0.0002 0.0098 0.1445 -0.9895>}
 ring_metacarpal_l {parent=hand_l, loc=<2.2074 -0.3625 1.3638>, rot=<-0.0284 0.0722 -0.0458 -0.9959>}
 ring_01_l {parent=ring_metacarpal_l, loc=<7.0295 -0.6197 -0.1446>, rot=<-0.0489 -0.0252 0.1892 -0.9804>}
 ring_02_l {parent=ring_01_l, loc=<3.6174 0.4647 -0.1027>, rot=<0 0 0.2806 -0.9598>}
 ring_03_l {parent=ring_02_l, loc=<2.4152 0.114 0.0199>, rot=<0 0 0.1189 -0.9929>}
 thumb_01_l {parent=hand_l, loc=<1.3118 -0.3191 -2.459>, rot=<0.6546 -0.4042 0.0664 -0.6354>}
 thumb_02_l {parent=thumb_01_l, loc=<5.9336 -0.0448 0.0401>, rot=<-0.0037 0.0213 0.298 -0.9543>}
 thumb_03_l {parent=thumb_02_l, loc=<2.2223 0.3682 -0.0907>, rot=<0.0118 -0.013 0.0066 -0.9998>}
 lowerarm_twist_01_l {parent=lowerarm_l, loc=<18.316 0 0>, rot=<0 0 0 1>}
 lowerarm_twist_02_l {parent=lowerarm_l, loc=<10.6707 0 0>, rot=<0 0 0 1>}
 upperarm_twist_01_l {parent=upperarm_l, loc=<7.2135 -0.1134 0>, rot=<0 0 0 1>}
 upperarm_twist_02_l {parent=upperarm_l, loc=<19.7998 -0.1134 0>, rot=<0 0 0 1>}
 dyn_shoulder_strap_l_master {parent=clavicle_l, loc=<5.0161 7.931 -0.3849>, rot=<0.2668 -0.4188 -0.4224 -0.7582>}
 dyn_shoulder_strap_l {parent=dyn_shoulder_strap_l_master, loc=<0 0 0>, rot=<0 0 0 1>}
 clavicle_r {parent=spine_05, loc=<5.4831 -0.1053 3.2946>, rot=<0.7289 0.0465 -0.6829 0.009>}
 upperarm_r {parent=clavicle_r, loc=<-9.8658 1.2136 0.4104>, rot=<0.0164 -0.3245 -0.0269 -0.9454>}
 lowerarm_r {parent=upperarm_r, loc=<-26.4651 0.8345 -0.259>, rot=<0.0036 0.0161 -0.2298 -0.9731>}
 hand_r {parent=lowerarm_r, loc=<-22.7356 0.3203 0.6106>, rot=<-0.6784 -0.0223 0.0735 -0.7307>}
 index_metacarpal_r {parent=hand_r, loc=<-1.6534 -0.0181 2.9601>, rot=<0.0158 -0.0124 -0.0173 -0.9996>}
 index_01_r {parent=index_metacarpal_r, loc=<-8.3008 -0.0357 0.0524>, rot=<-0.0044 0.0485 0.0924 -0.9945>}
 index_02_r {parent=index_01_r, loc=<-3.3589 0.3003 -0.0453>, rot=<-0.002 0.0027 0.3225 -0.9466>}
 index_03_r {parent=index_02_r, loc=<-2.5578 0.123 0.0252>, rot=<-0.002 0.0067 0.2558 -0.9667>}
 middle_metacarpal_r {parent=hand_r, loc=<-2.1856 0.1726 0.5984>, rot=<0.006 0.0168 -0.0329 -0.9993>}
 middle_01_r {parent=middle_metacarpal_r, loc=<-7.3273 0.2447 -0.0129>, rot=<-0.0208 0.0042 0.0858 -0.9961>}
 middle_02_r {parent=middle_01_r, loc=<-4.0124 -0.14 -0.1175>, rot=<-0.0218 0.0258 0.3512 -0.9357>}
 middle_03_r {parent=middle_02_r, loc=<-2.6077 -0.0533 -0.1373>, rot=<0.0082 0.0188 0.1708 -0.9851>}
 pinky_metacarpal_r {parent=hand_r, loc=<-1.7617 0.9753 -3.3814>, rot=<-0.0397 0.0814 -0.069 -0.9935>}
 pinky_01_r {parent=pinky_metacarpal_r, loc=<-7.1707 0.4832 0.1094>, rot=<-0.0654 -0.0035 0.2242 -0.9723>}
 pinky_02_r {parent=pinky_01_r, loc=<-2.8274 -0.0892 0.1043>, rot=<0.0037 0.0011 0.2854 -0.9584>}
 pinky_03_r {parent=pinky_02_r, loc=<-2.2045 -0.031 0.0078>, rot=<-0.0002 0.0098 0.1445 -0.9895>}
 ring_metacarpal_r {parent=hand_r, loc=<-2.2073 0.3627 -1.3638>, rot=<-0.0284 0.0722 -0.0458 -0.9959>}
 ring_01_r {parent=ring_metacarpal_r, loc=<-7.0297 0.6197 0.1446>, rot=<-0.0489 -0.0252 0.1892 -0.9804>}
 ring_02_r {parent=ring_01_r, loc=<-3.6171 -0.4648 0.1027>, rot=<0 0 0.2806 -0.9598>}
 ring_03_r {parent=ring_02_r, loc=<-2.4153 -0.1141 -0.0199>, rot=<0 0 0.1189 -0.9929>}
 thumb_01_r {parent=hand_r, loc=<-1.3119 0.3193 2.459>, rot=<0.6546 -0.4042 0.0664 -0.6354>}
 thumb_02_r {parent=thumb_01_r, loc=<-5.9335 0.0447 -0.0401>, rot=<-0.0037 0.0213 0.298 -0.9543>}
 thumb_03_r {parent=thumb_02_r, loc=<-2.2223 -0.3682 0.0907>, rot=<0.0118 -0.013 0.0066 -0.9998>}
 lowerarm_twist_01_r {parent=lowerarm_r, loc=<-18.3161 0 0>, rot=<0 0 0 1>}
 lowerarm_twist_02_r {parent=lowerarm_r, loc=<-10.6707 0 0>, rot=<0 0 0 1>}
 upperarm_twist_01_r {parent=upperarm_r, loc=<-7.2135 0.1134 0>, rot=<0 0 0 1>}
 upperarm_twist_02_r {parent=upperarm_r, loc=<-19.7998 0.1134 0>, rot=<0 0 0 1>}
 dyn_ribbon_master {parent=upperarm_r, loc=<-14.7849 0.3327 -4.275>, rot=<0.0476 -0.9979 0.0335 0.0286>}
 dyn_ribbon_base {parent=dyn_ribbon_master, loc=<0 0 0>, rot=<0 0 0 -1>}
 dyn_ribbon_bow_back_01 {parent=dyn_ribbon_base, loc=<-0.2612 0.9108 -0.5217>, rot=<0 0 -0.6923 -0.7216>}
 dyn_ribbon_bow_back_02 {parent=dyn_ribbon_bow_back_01, loc=<1.9702 0 -0.7269>, rot=<0 -0.3063 0 -0.952>}
 dyn_ribbon_bow_front_01 {parent=dyn_ribbon_base, loc=<-0.2284 -0.5255 -0.3176>, rot=<0 0 0.702 -0.7122>}
 dyn_ribbon_bow_front_02 {parent=dyn_ribbon_bow_front_01, loc=<2.4823 -0.1759 -0.2417>, rot=<-0.0953 -0.2399 -0.1029 -0.9606>}
 dyn_ribbon_strip_back_01 {parent=dyn_ribbon_base, loc=<0.9378 0.3392 -0.3295>, rot=<0.02 0.0827 -0.1052 -0.9908>}
 dyn_ribbon_strip_back_02 {parent=dyn_ribbon_strip_back_01, loc=<3.0094 0.1129 -0.0603>, rot=<0 0 0 -1>}
 dyn_ribbon_strip_front_01 {parent=dyn_ribbon_base, loc=<0.4381 -0.7213 -0.2784>, rot=<-0.0054 0.0849 0.1929 -0.9775>}
 dyn_ribbon_strip_front_02 {parent=dyn_ribbon_strip_front_01, loc=<2.4678 0.0894 -0.3837>, rot=<-0.0993 0 0 -0.9951>}
 dyn_shoulder_strap_r_master {parent=clavicle_r, loc=<-5.0108 -7.9635 0.4687>, rot=<0.4098 -0.7635 0.2653 0.4228>}
 dyn_shoulder_strap_r {parent=dyn_shoulder_strap_r_master, loc=<0 0 0>, rot=<0 0 0 1>}
 neck_01 {parent=spine_05, loc=<11.3138 0.1529 0>, rot=<0 0 -0.2003 -0.9797>}
 neck_02 {parent=neck_01, loc=<4.5627 0 0>, rot=<0 0 0.0504 -0.9987>}
 head {parent=neck_02, loc=<5.3914 0 0>, rot=<0 0 0.115 -0.9934>}
 dyn_shirt_base {parent=spine_03, loc=<1.8341 0.2657 0>, rot=<0 0 0 1>}
 dyn_shirt_back_r_master {parent=dyn_shirt_base, loc=<-1.2433 -5.8145 4.2436>, rot=<0.0083 -0.574 0.8188 -0.0077>}
 dyn_shirt_back_r {parent=dyn_shirt_back_r_master, loc=<0 0 0>, rot=<0 0 0 1>}
 dyn_shirt_back_mid_r_master {parent=dyn_shirt_base, loc=<-1.587 -0.7456 9.6771>, rot=<0.0359 -0.1468 0.9873 -0.0499>}
 dyn_shirt_back_mid_r {parent=dyn_shirt_back_mid_r_master, loc=<0 0 0>, rot=<0 0 0 1>}
 dyn_shirt_front_mid_r_master {parent=dyn_shirt_base, loc=<-1.8277 6.3467 8.2906>, rot=<-0.0086 0.3211 0.9461 -0.0421>}
 dyn_shirt_front_mid_r {parent=dyn_shirt_front_mid_r_master, loc=<0 0 0>, rot=<0 0 0 -1>}
 dyn_shirt_front_r_master {parent=dyn_shirt_base, loc=<-1.9651 10.4886 3.7764>, rot=<-0.0264 0.5653 0.8239 -0.0304>}
 dyn_shirt_front_r {parent=dyn_shirt_front_r_master, loc=<0 0 0>, rot=<0 0 0 -1>}
 dyn_shirt_back_l_master {parent=dyn_shirt_base, loc=<-1.2438 -5.8145 -4.2436>, rot=<0.0075 -0.8061 0.5917 -0.0085>}
 dyn_shirt_back_l {parent=dyn_shirt_back_l_master, loc=<0 0 0>, rot=<0 0 0 1>}
 dyn_shirt_back_mid_l_master {parent=dyn_shirt_base, loc=<-1.5869 -0.7456 -9.6771>, rot=<0.0472 -0.9741 0.2175 -0.0394>}
 dyn_shirt_back_mid_l {parent=dyn_shirt_back_mid_l_master, loc=<0 0 0>, rot=<0 0 0 1>}
 dyn_shirt_front_mid_l_master {parent=dyn_shirt_base, loc=<-1.8276 6.3467 -8.2906>, rot=<0.0562 -0.9897 -0.1314 -0.0074>}
 dyn_shirt_front_mid_l {parent=dyn_shirt_front_mid_l_master, loc=<0 0 0>, rot=<0 0 0 -1>}
 dyn_shirt_front_l_master {parent=dyn_shirt_base, loc=<-1.9651 10.4886 -3.7764>, rot=<0.0534 -0.859 -0.5088 0.0172>}
 dyn_shirt_front_l {parent=dyn_shirt_front_l_master, loc=<0 0 0>, rot=<0 0 0 -1>}
 dyn_jacket_base {parent=spine_03, loc=<4.7701 0.2657 0>, rot=<0 0 0 1>}
 dyn_jacket_r_front_master {parent=dyn_jacket_base, loc=<0.9078 10.0213 8.7599>, rot=<-0.0223 0.4955 0.8669 -0.0503>}
 dyn_jacket_r_front {parent=dyn_jacket_r_front_master, loc=<0 0 0>, rot=<0 0 0 -1>}
 dyn_jacket_r_mid_master {parent=dyn_jacket_base, loc=<0.3137 1.1644 11.8073>, rot=<-0.0088 -0.0005 0.9985 -0.0547>}
 dyn_jacket_r_mid {parent=dyn_jacket_r_mid_master, loc=<0 0 0>, rot=<0 0 0 1>}
 dyn_jacket_r_back_master {parent=dyn_jacket_base, loc=<0.4707 -6.4115 7.6491>, rot=<0.0141 -0.5349 0.8441 -0.0341>}
 dyn_jacket_r_back {parent=dyn_jacket_r_back_master, loc=<0 0 0>, rot=<0 0 0 1>}
 dyn_jacket_back_master {parent=dyn_jacket_base, loc=<0.3836 -8.3298 -0.0214>, rot=<0.0302 -0.7063 0.707 -0.0212>}
 dyn_jacket_back {parent=dyn_jacket_back_master, loc=<0 0 0>, rot=<0 0 0 -1>}
 dyn_jacket_l_back_master {parent=dyn_jacket_base, loc=<0.4704 -6.4115 -7.6491>, rot=<0.0347 -0.8669 0.4971 -0.0126>}
 dyn_jacket_l_back {parent=dyn_jacket_l_back_master, loc=<0 0 0>, rot=<0 0 0 -1>}
 dyn_jacket_l_mid_master {parent=dyn_jacket_base, loc=<0.3133 1.1644 -11.8073>, rot=<0.0547 -0.9985 0.0005 0.0088>}
 dyn_jacket_l_mid {parent=dyn_jacket_l_mid_master, loc=<0 0 0>, rot=<0 0 0 1>}
 dyn_jacket_l_front_master {parent=dyn_jacket_base, loc=<0.9078 10.0213 -8.7599>, rot=<0.0527 -0.9227 -0.3816 0.0157>}
 dyn_jacket_l_front {parent=dyn_jacket_l_front_master, loc=<0 0 0>, rot=<0 0 0 1>}
 dyn_pouch_strap_l_master {parent=dyn_jacket_base, loc=<-0.3438 11.3798 -11.2784>, rot=<-0.0069 -0.9386 -0.3424 -0.0419>}
 dyn_pouch_strap_l {parent=dyn_pouch_strap_l_master, loc=<0 0 0>, rot=<0 0 0 -1>}
 dyn_pouch_strap_r_master {parent=dyn_jacket_base, loc=<-0.4741 11.3422 11.313>, rot=<0.0517 0.3408 0.9386 0.0138>}
 dyn_pouch_strap_r {parent=dyn_pouch_strap_r_master, loc=<0 0 0>, rot=<0 0 0 1>}
 thigh_l {parent=pelvis, loc=<-1.0855 -0.2109 -8.7053>, rot=<0.0009 0.0355 0.0155 -0.9993>}
 calf_l {parent=thigh_l, loc=<-42.5284 -0.0019 -0.7297>, rot=<0.0006 -0.0161 -0.0391 -0.9991>}
 calf_twist_01_l {parent=calf_l, loc=<-35.8483 -1.9197 -0.5794>, rot=<0.0237 0.0095 -0.0035 -0.9997>}
 calf_twist_02_l {parent=calf_l, loc=<-19.1546 -1.7943 -0.2638>, rot=<0.0237 0.0095 -0.0035 -0.9997>}
 foot_l {parent=calf_l, loc=<-46.7401 -2.5014 -0.592>, rot=<0.0427 -0.0238 0.0353 -0.9982>}
 ball_l {parent=foot_l, loc=<-3.5658 10.8066 -0.1544>, rot=<0.0001 -0.0001 -0.7063 -0.7079>}
 thigh_twist_01_l {parent=thigh_l, loc=<-27.952 0 0>, rot=<0.0013 0 0 -1>}
 thigh_r {parent=pelvis, loc=<-1.0853 -0.2109 8.7053>, rot=<0.0355 -0.0009 0.9993 0.0155>}
 calf_r {parent=thigh_r, loc=<42.5286 0.0019 0.7297>, rot=<0.0006 -0.0161 -0.0391 -0.9991>}
 calf_twist_01_r {parent=calf_r, loc=<35.8485 1.9197 0.5793>, rot=<0.0237 0.0095 -0.0035 -0.9997>}
 calf_twist_02_r {parent=calf_r, loc=<19.1546 1.7943 0.2638>, rot=<0.0237 0.0095 -0.0035 -0.9997>}
 foot_r {parent=calf_r, loc=<46.7403 2.5014 0.5919>, rot=<0.0427 -0.0238 0.0353 -0.9982>}
 ball_r {parent=foot_r, loc=<3.5658 -10.8066 0.1544>, rot=<0.0001 -0.0001 -0.7063 -0.7079>}
 thigh_twist_01_r {parent=thigh_r, loc=<27.952 0 0>, rot=<0.0013 0 0 -1>}
 dyn_pistol_master {parent=pelvis, loc=<1.8695 1.1894 14.601>, rot=<0.1102 0.1304 0.9843 -0.0443>}
 dyn_pistol {parent=dyn_pistol_master, loc=<0 0 0>, rot=<0 0 0 -1>}
]
