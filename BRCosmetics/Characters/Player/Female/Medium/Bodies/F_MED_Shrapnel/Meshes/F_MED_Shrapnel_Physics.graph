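# Physics asset for the F_MED_Shrapnel female player body cosmetic, giving capsule bodies to a dangling ribbon and a pistol for secondary motion.
# Four two-bone ribbon chains, front and back bow and strip, branch from a ribbon base with tight swing and twist limits.
# The base joins the right upper arm capsule with free rotation.
# The pistol hangs separately from its master bone, swinging widest on one axis.
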PhysicsAsset /BRCosmetics/Characters/Player/Female/Medium/Bodies/F_MED_Shrapnel/Meshes/F_MED_Shrapnel_Physics
bodies {
 dyn_pistol [{shapeType=Capsule, length=18.2729, radius=2, center=<0.3695 0 1.3497>, rotation=<rot 90 0 0>}]
 dyn_pistol_master [{shapeType=Capsule, length=2, radius=2, rotation=<rot 90 0 0>}]
 dyn_ribbon_base [{shapeType=Capsule, length=1, radius=1}]
 dyn_ribbon_bow_back_01 [{shapeType=Capsule, length=2.4885, radius=0.9604, center=<1.3461 0 0>, rotation=<rot 0 0 90>}]
 dyn_ribbon_bow_back_02 [{shapeType=Capsule, length=1.9368, radius=1.5122, center=<1.3461 0 0>, rotation=<rot 0 0 90>}]
 dyn_ribbon_bow_front_01 [{shapeType=Capsule, length=2.6107, radius=0.5757, center=<1.278 0 -0.0007>, rotation=<rot 0 0 90>}]
 dyn_ribbon_bow_front_02 [{shapeType=Capsule, length=4.7284, radius=0.7159, center=<1.278 0 -0.0007>, rotation=<rot 0 0 90>}]
 dyn_ribbon_strip_back_01 [{shapeType=Capsule, length=1.2746, radius=0.5757, center=<1.278 0 -0.0007>, rotation=<rot 0 0 90>}]
 dyn_ribbon_strip_back_02 [{shapeType=Capsule, length=2.388, radius=0.7984, center=<1.8831 0 -0.0006>, rotation=<rot 0 0 90>}]
 dyn_ribbon_strip_front_01 [{shapeType=Capsule, length=1.2311, radius=0.5757, center=<1.278 0 -0.0007>, rotation=<rot 0 0 90>}]
 dyn_ribbon_strip_front_02 [{shapeType=Capsule, length=1.5028, radius=0.7409, center=<1.7582 0 -0.0007>, rotation=<rot 0 0 90>}]
 upperarm_r [{shapeType=Capsule, length=18.8951, radius=3.6046, center=<-13.1345 -0.0001 0>, rotation=<rot 90 0 0>}]
}
constraints [
 {bone1Name=dyn_ribbon_bow_back_01, bone2Name=dyn_ribbon_base, swing1LimitDegrees=10, swing2LimitDegrees=10, twistLimitDegrees=8}
 {bone1Name=dyn_ribbon_bow_back_02, bone2Name=dyn_ribbon_bow_back_01, swing1LimitDegrees=10, swing2LimitDegrees=10, twistLimitDegrees=8}
 {bone1Name=dyn_ribbon_bow_front_01, bone2Name=dyn_ribbon_base, swing1LimitDegrees=10, swing2LimitDegrees=10, twistLimitDegrees=5}
 {bone1Name=dyn_ribbon_bow_front_02, bone2Name=dyn_ribbon_bow_front_01, swing1LimitDegrees=10, swing2LimitDegrees=10, twistLimitDegrees=5}
 {bone1Name=dyn_ribbon_strip_back_01, bone2Name=dyn_ribbon_base, swing1LimitDegrees=10, swing2LimitDegrees=10, twistLimitDegrees=5}
 {bone1Name=dyn_ribbon_strip_back_02, bone2Name=dyn_ribbon_strip_back_01, swing1LimitDegrees=10, swing2LimitDegrees=10, twistLimitDegrees=5}
 {bone1Name=dyn_ribbon_strip_front_01, bone2Name=dyn_ribbon_base, swing1LimitDegrees=10, swing2LimitDegrees=10, twistLimitDegrees=5}
 {bone1Name=dyn_ribbon_strip_front_02, bone2Name=dyn_ribbon_strip_front_01, swing1LimitDegrees=10, swing2LimitDegrees=10, twistLimitDegrees=5}
 {
  bone1Name=dyn_ribbon_base
  bone2Name=upperarm_r
  swing1LimitDegrees=45
  swing1Motion=Free
  swing2LimitDegrees=45
  swing2Motion=Free
  twistLimitDegrees=45
  twistMotion=Free
 }
 {bone1Name=dyn_pistol, bone2Name=dyn_pistol_master, swing1LimitDegrees=50, swing2LimitDegrees=10, twistLimitDegrees=3}
]
