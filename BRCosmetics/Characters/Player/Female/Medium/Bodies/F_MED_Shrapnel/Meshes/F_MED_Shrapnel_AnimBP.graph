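# Secondary-motion AnimBP for the F_MED_Shrapnel female character body, animating jacket, shirt, collar and straps over the pose copied from the attached parent mesh.
# Per-bone AnimDynamics simulate cloth and accessory bones, interleaved with weighted constraints placing jacket, strap and pistol master bones.
# A RigidBody pass using the Shrapnel physics asset then feeds a Control Rig before output.
# Several stray nodes and the EventGraph debug logic are unconnected.
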AnimBlueprint /BRCosmetics/Characters/Player/Female/Medium/Bodies/F_MED_Shrapnel/Meshes/F_MED_Shrapnel_AnimBP
target_skeleton=&/F_MED_Shrapnel_Skeleton
graphs [
 AnimGraph {nodes [
   Anim Dynamics - Bone: dyn_pistol(Alpha=0.7, ExternalForce="0,0,0") {
    angularDampingOverride=0.9
    bGravityOverrideInSimSpace=true
    bOverrideAngularDamping=true
    bUseGravityOverride=true
    simulationSpace=BoneRelative
    alphaBoolBlend {blendOption=Linear}
    boundBone {boneName=dyn_pistol}
    componentAppliedLinearAccClamp=<1500 1500 1500>
    componentLinearAccScale=<0.2 0.2 0.3>
    gravityOverride {z=-980}
    physicsBodyDefinitions [{
      sphereCollisionRadius=0
      boundBone {boneName=dyn_pistol}
      boxExtents=<15 15 15>
      constraintSetup {angularLimitsMax=<0 -3 8>, angularLimitsMin=<0 -8 -8>, angularTarget=<1 0 0>}
      localJointOffset=<2 0 0>
     }]
    relativeSpaceBone {boneName=root}
   }
   n1 = "Constraint - dyn_shoulder_strap_l_master (PPPPP)"!(
    ComponentPose
    Alpha=1
    ConstraintWeights_0=0
    ConstraintWeights_1:float=0.2654
    ConstraintWeights_2:float=0.5057
    ConstraintWeights_3:float=0.016
    ConstraintWeights_4:float=0.2129
   ) -> (Pose) {
    alphaBoolBlend {blendOption=Linear}
    boneToModify {boneName=dyn_shoulder_strap_l_master}
    constraintSetup [
     {targetBone {boneName=spine_03}}
     {targetBone {boneName=spine_04}}
     {targetBone {boneName=spine_05}}
     {targetBone {boneName=upperarm_twist_01_l}}
     {targetBone {boneName=clavicle_l}}
    ]
    constraintWeights [1,1,1,1,1]
   }
   "Constraint - dyn_shoulder_strap_r_master (PPP)"!(ComponentPose=@n1, Alpha=1, ConstraintWeights_0=0.166, ConstraintWeights_1:float=0.3752, ConstraintWeights_2:float=0.4588) -> (Pose) {
    alphaBoolBlend {blendOption=Linear}
    boneToModify {boneName=dyn_shoulder_strap_r_master}
    constraintSetup [{targetBone {boneName=clavicle_r}}, {targetBone {boneName=spine_04}}, {targetBone {boneName=spine_05}}]
    constraintWeights [1,1,1]
   }
   "Constraint - dyn_ribbon_strip_back_01 (T)"(Alpha=1, ConstraintWeights_0=1) {
    alphaBoolBlend {blendOption=Linear}
    boneToModify {boneName=dyn_ribbon_strip_back_01}
    constraintSetup [{transformType=Translation, targetBone {boneName=dyn_ribbon_base}}]
    constraintWeights [1]
   }
   n2 = "Constraint - dyn_jacket_r_front_master (PPPPPPPPPPPPPPPPPPPPPPPPPP)"!(
    ComponentPose=Local To Component(LocalPose=Copy Pose From Mesh() {bUseAttachedParent=true})
    Alpha=1
    ConstraintWeights_0=0
    ConstraintWeights_1:float=0.0079
    ConstraintWeights_2:float=0
    ConstraintWeights_3:float=0
    ConstraintWeights_4:float=0.0006
    ConstraintWeights_5:float=0
    ConstraintWeights_6:float=0.2902
    ConstraintWeights_7:float=0.0074
    ConstraintWeights_8:float=0.0109
    ConstraintWeights_9:float=0.0001
    ConstraintWeights_10:float=0
    ConstraintWeights_11:float=0
    ConstraintWeights_12:float=0.0491
    ConstraintWeights_13:float=0.5468
    ConstraintWeights_14:float=0.0001
    ConstraintWeights_15:float=0.0035
    ConstraintWeights_16:float=0
    ConstraintWeights_17:float=0.0001
    ConstraintWeights_18:float=0.0002
    ConstraintWeights_19:float=0.0745
    ConstraintWeights_20:float=0
    ConstraintWeights_21:float=0.0003
    ConstraintWeights_22:float=0.0024
    ConstraintWeights_23:float=0.0059
    ConstraintWeights_24:float=0
    ConstraintWeights_25:float=0
   ) -> (Pose) {
    alphaBoolBlend {blendOption=Linear}
    boneToModify {boneName=dyn_jacket_r_front_master}
    constraintSetup [
     {targetBone {boneName=dyn_shirt_back_mid_l}}
     {targetBone {boneName=dyn_jacket_r_mid}}
     {targetBone {boneName=dyn_shirt_back_l}}
     {targetBone {boneName=spine_01}}
     {targetBone {boneName=spine_02}}
     {targetBone {boneName=dyn_jacket_back}}
     {targetBone {boneName=spine_04}}
     {targetBone {boneName=spine_05}}
     {targetBone {boneName=dyn_shirt_back_mid_r}}
     {targetBone {boneName=dyn_shirt_front_mid_l}}
     {targetBone {boneName=clavicle_l}}
     {targetBone {boneName=upperarm_twist_01_r}}
     {targetBone {boneName=dyn_shirt_front_r}}
     {targetBone {boneName=dyn_jacket_r_front}}
     {targetBone {boneName=dyn_jacket_l_front}}
     {targetBone {boneName=dyn_shirt_front_l}}
     {targetBone {boneName=upperarm_twist_01_l}}
     {targetBone {boneName=clavicle_r}}
     {targetBone {boneName=dyn_jacket_r_back}}
     {targetBone {boneName=dyn_shirt_front_mid_r}}
     {targetBone {boneName=dyn_jacket_l_back}}
     {targetBone {boneName=dyn_shirt_back_r}}
     {targetBone {boneName=dyn_pouch_strap_r}}
     {targetBone {boneName=spine_03}}
     {targetBone {boneName=dyn_pouch_strap_l}}
     {targetBone {boneName=dyn_jacket_l_mid}}
    ]
    constraintWeights [1,1,1,1,1,1,1,1,1,1,1,1,1,1,1,1,1,1,1,1,1,1,1,1,1,1]
   }
   n3 = "Constraint - dyn_jacket_l_front_master (PPPPPPPPPPPPPPPPPPPPPPPPPP)"!(
    ComponentPose=@n2
    Alpha=1
    ConstraintWeights_0=0.0098
    ConstraintWeights_1:float=0
    ConstraintWeights_2:float=0.0002
    ConstraintWeights_3:float=0.0001
    ConstraintWeights_4:float=0.0009
    ConstraintWeights_5:float=0
    ConstraintWeights_6:float=0.325
    ConstraintWeights_7:float=0.0011
    ConstraintWeights_8:float=0
    ConstraintWeights_9:float=0.067
    ConstraintWeights_10:float=0.0002
    ConstraintWeights_11:float=0
    ConstraintWeights_12:float=0.0037
    ConstraintWeights_13:float=0.0001
    ConstraintWeights_14:float=0.5209
    ConstraintWeights_15:float=0.0541
    ConstraintWeights_16:float=0
    ConstraintWeights_17:float=0
    ConstraintWeights_18:float=0
    ConstraintWeights_19:float=0.0001
    ConstraintWeights_20:float=0.0004
    ConstraintWeights_21:float=0
    ConstraintWeights_22:float=0
    ConstraintWeights_23:float=0.0073
    ConstraintWeights_24:float=0.0011
    ConstraintWeights_25:float=0.0082
   ) -> (Pose) {
    alphaBoolBlend {blendOption=Linear}
    boneToModify {boneName=dyn_jacket_l_front_master}
    constraintSetup [
     {targetBone {boneName=dyn_shirt_back_mid_l}}
     {targetBone {boneName=dyn_jacket_r_mid}}
     {targetBone {boneName=dyn_shirt_back_l}}
     {targetBone {boneName=spine_01}}
     {targetBone {boneName=spine_02}}
     {targetBone {boneName=dyn_jacket_back}}
     {targetBone {boneName=spine_04}}
     {targetBone {boneName=spine_05}}
     {targetBone {boneName=dyn_shirt_back_mid_r}}
     {targetBone {boneName=dyn_shirt_front_mid_l}}
     {targetBone {boneName=clavicle_l}}
     {targetBone {boneName=upperarm_twist_01_r}}
     {targetBone {boneName=dyn_shirt_front_r}}
     {targetBone {boneName=dyn_jacket_r_front}}
     {targetBone {boneName=dyn_jacket_l_front}}
     {targetBone {boneName=dyn_shirt_front_l}}
     {targetBone {boneName=upperarm_twist_01_l}}
     {targetBone {boneName=clavicle_r}}
     {targetBone {boneName=dyn_jacket_r_back}}
     {targetBone {boneName=dyn_shirt_front_mid_r}}
     {targetBone {boneName=dyn_jacket_l_back}}
     {targetBone {boneName=dyn_shirt_back_r}}
     {targetBone {boneName=dyn_pouch_strap_r}}
     {targetBone {boneName=spine_03}}
     {targetBone {boneName=dyn_pouch_strap_l}}
     {targetBone {boneName=dyn_jacket_l_mid}}
    ]
    constraintWeights [1,1,1,1,1,1,1,1,1,1,1,1,1,1,1,1,1,1,1,1,1,1,1,1,1,1]
   }
   n4 = Anim Dynamics - Bone: dyn_jacket_r_front(ComponentPose=@n3, Alpha=1, ExternalForce="0,0,0") {
    angularDampingOverride=200
    angularSpringConstant=500
    bGravityOverrideInSimSpace=true
    bUseGravityOverride=true
    simulationSpace=BoneRelative
    alphaBoolBlend {blendOption=Linear}
    boundBone {boneName=dyn_jacket_r_front}
    componentAppliedLinearAccClamp=<1500 1500 1500>
    componentLinearAccScale=<0.4 0.4 1>
    componentLinearVelScale=<0.2 0.2 0.4>
    gravityOverride {z=-980}
    physicsBodyDefinitions [{
      sphereCollisionRadius=0
      boundBone {boneName=dyn_jacket_r_front}
      boxExtents=<30 30 30>
      constraintSetup {angularLimitsMax=<2 2 2>, angularLimitsMin=<-2 -2 -2>, angularTarget=<1 0 0>}
      localJointOffset=<8 0 0>
     }]
    relativeSpaceBone {boneName=root}
   }
   n5 = Anim Dynamics - Bone: dyn_jacket_r_mid(ComponentPose=@n4, Alpha=1, ExternalForce="0,0,0") {
    angularDampingOverride=200
    angularSpringConstant=500
    bGravityOverrideInSimSpace=true
    bUseGravityOverride=true
    simulationSpace=BoneRelative
    alphaBoolBlend {blendOption=Linear}
    boundBone {boneName=dyn_jacket_r_mid}
    componentAppliedLinearAccClamp=<1500 1500 1500>
    componentLinearAccScale=<0.4 0.4 1>
    componentLinearVelScale=<0.2 0.2 0.4>
    gravityOverride {z=-980}
    physicsBodyDefinitions [{
      sphereCollisionRadius=0
      boundBone {boneName=dyn_jacket_r_mid}
      boxExtents=<30 30 30>
      constraintSetup {angularLimitsMax=<5 5 5>, angularLimitsMin=<-5 -5 -5>, angularTarget=<1 0 0>}
      localJointOffset=<8 0 0>
     }]
    relativeSpaceBone {boneName=root}
   }
   n6 = Anim Dynamics - Bone: dyn_jacket_r_back(ComponentPose=@n5, Alpha=1, ExternalForce="0,0,0") {
    angularDampingOverride=200
    angularSpringConstant=500
    bGravityOverrideInSimSpace=true
    bUseGravityOverride=true
    simulationSpace=BoneRelative
    alphaBoolBlend {blendOption=Linear}
    boundBone {boneName=dyn_jacket_r_back}
    componentAppliedLinearAccClamp=<1500 1500 1500>
    componentLinearAccScale=<0.4 0.4 1>
    componentLinearVelScale=<0.2 0.2 0.4>
    gravityOverride {z=-980}
    physicsBodyDefinitions [{
      sphereCollisionRadius=0
      boundBone {boneName=dyn_jacket_r_back}
      boxExtents=<30 30 30>
      constraintSetup {angularLimitsMax=<5 5 5>, angularLimitsMin=<-5 -5 -5>, angularTarget=<1 0 0>}
      localJointOffset=<8 0 0>
     }]
    relativeSpaceBone {boneName=root}
   }
   n7 = Anim Dynamics - Bone: dyn_jacket_back(ComponentPose=@n6, Alpha=1, ExternalForce="0,0,0") {
    angularDampingOverride=200
    angularSpringConstant=500
    bGravityOverrideInSimSpace=true
    bUseGravityOverride=true
    simulationSpace=BoneRelative
    alphaBoolBlend {blendOption=Linear}
    boundBone {boneName=dyn_jacket_back}
    componentAppliedLinearAccClamp=<1500 1500 1500>
    componentLinearAccScale=<0.4 0.4 1>
    componentLinearVelScale=<0.2 0.2 0.4>
    gravityOverride {z=-980}
    physicsBodyDefinitions [{
      sphereCollisionRadius=0
      boundBone {boneName=dyn_jacket_back}
      boxExtents=<30 30 30>
      constraintSetup {angularLimitsMax=<5 5 5>, angularLimitsMin=<-5 -5 -5>, angularTarget=<1 0 0>}
      localJointOffset=<8 0 0>
     }]
    relativeSpaceBone {boneName=root}
   }
   n8 = Anim Dynamics - Bone: dyn_jacket_l_back(ComponentPose=@n7, Alpha=1, ExternalForce="0,0,0") {
    angularDampingOverride=200
    angularSpringConstant=500
    bGravityOverrideInSimSpace=true
    bUseGravityOverride=true
    simulationSpace=BoneRelative
    alphaBoolBlend {blendOption=Linear}
    boundBone {boneName=dyn_jacket_l_back}
    componentAppliedLinearAccClamp=<1500 1500 1500>
    componentLinearAccScale=<0.4 0.4 1>
    componentLinearVelScale=<0.2 0.2 0.4>
    gravityOverride {z=-980}
    physicsBodyDefinitions [{
      sphereCollisionRadius=0
      boundBone {boneName=dyn_jacket_l_back}
      boxExtents=<30 30 30>
      constraintSetup {angularLimitsMax=<5 5 5>, angularLimitsMin=<-5 -5 -5>, angularTarget=<1 0 0>}
      localJointOffset=<8 0 0>
     }]
    relativeSpaceBone {boneName=root}
   }
   n9 = Anim Dynamics - Bone: dyn_jacket_l_mid(ComponentPose=@n8, Alpha=1, ExternalForce="0,0,0") {
    angularDampingOverride=200
    angularSpringConstant=500
    bGravityOverrideInSimSpace=true
    bUseGravityOverride=true
    simulationSpace=BoneRelative
    alphaBoolBlend {blendOption=Linear}
    boundBone {boneName=dyn_jacket_l_mid}
    componentAppliedLinearAccClamp=<1500 1500 1500>
    componentLinearAccScale=<0.4 0.4 1>
    componentLinearVelScale=<0.2 0.2 0.4>
    gravityOverride {z=-980}
    physicsBodyDefinitions [{
      sphereCollisionRadius=0
      boundBone {boneName=dyn_jacket_l_mid}
      boxExtents=<30 30 30>
      constraintSetup {angularLimitsMax=<5 5 5>, angularLimitsMin=<-5 -5 -5>, angularTarget=<1 0 0>}
      localJointOffset=<8 0 0>
     }]
    relativeSpaceBone {boneName=root}
   }
   n10 = Anim Dynamics - Bone: dyn_jacket_l_front(ComponentPose=@n9, Alpha=1, ExternalForce="0,0,0") {
    angularDampingOverride=200
    angularSpringConstant=500
    bGravityOverrideInSimSpace=true
    bUseGravityOverride=true
    simulationSpace=BoneRelative
    alphaBoolBlend {blendOption=Linear}
    boundBone {boneName=dyn_jacket_l_front}
    componentAppliedLinearAccClamp=<1500 1500 1500>
    componentLinearAccScale=<0.4 0.4 1>
    componentLinearVelScale=<0.2 0.2 0.4>
    gravityOverride {z=-980}
    physicsBodyDefinitions [{
      sphereCollisionRadius=0
      boundBone {boneName=dyn_jacket_l_front}
      boxExtents=<30 30 30>
      constraintSetup {angularLimitsMax=<2 2 2>, angularLimitsMin=<-2 -2 -2>, angularTarget=<1 0 0>}
      localJointOffset=<8 0 0>
     }]
    relativeSpaceBone {boneName=root}
   }
   n11 = Anim Dynamics - Bone: dyn_shirt_back_r(ComponentPose=@n10, Alpha=1, ExternalForce="0,0,0") {
    angularDampingOverride=200
    angularSpringConstant=500
    bGravityOverrideInSimSpace=true
    bUseGravityOverride=true
    simulationSpace=BoneRelative
    alphaBoolBlend {blendOption=Linear}
    boundBone {boneName=dyn_shirt_back_r}
    componentAppliedLinearAccClamp=<1500 1500 1500>
    componentLinearAccScale=<0.4 0.4 1>
    componentLinearVelScale=<0.2 0.2 0.4>
    gravityOverride {z=-980}
    physicsBodyDefinitions [{
      sphereCollisionRadius=0
      boundBone {boneName=dyn_shirt_back_r}
      boxExtents=<30 30 30>
      constraintSetup {angularLimitsMax=<5 5 5>, angularLimitsMin=<-5 -5 -5>, angularTarget=<1 0 0>}
      localJointOffset=<8 0 0>
     }]
    relativeSpaceBone {boneName=root}
   }
   n12 = Anim Dynamics - Bone: dyn_shirt_back_mid_r(ComponentPose=@n11, Alpha=1, ExternalForce="0,0,0") {
    angularDampingOverride=200
    angularSpringConstant=500
    bGravityOverrideInSimSpace=true
    bUseGravityOverride=true
    simulationSpace=BoneRelative
    alphaBoolBlend {blendOption=Linear}
    boundBone {boneName=dyn_shirt_back_mid_r}
    componentAppliedLinearAccClamp=<1500 1500 1500>
    componentLinearAccScale=<0.4 0.4 1>
    componentLinearVelScale=<0.2 0.2 0.4>
    gravityOverride {z=-980}
    physicsBodyDefinitions [{
      sphereCollisionRadius=0
      boundBone {boneName=dyn_shirt_back_mid_r}
      boxExtents=<30 30 30>
      constraintSetup {angularLimitsMax=<5 5 5>, angularLimitsMin=<-5 -5 -5>, angularTarget=<1 0 0>}
      localJointOffset=<8 0 0>
     }]
    relativeSpaceBone {boneName=root}
   }
   n13 = Anim Dynamics - Bone: dyn_shirt_front_mid_r(ComponentPose=@n12, Alpha=1, ExternalForce="0,0,0") {
    angularDampingOverride=200
    angularSpringConstant=500
    bGravityOverrideInSimSpace=true
    bUseGravityOverride=true
    simulationSpace=BoneRelative
    alphaBoolBlend {blendOption=Linear}
    boundBone {boneName=dyn_shirt_front_mid_r}
    componentAppliedLinearAccClamp=<1500 1500 1500>
    componentLinearAccScale=<0.4 0.4 1>
    componentLinearVelScale=<0.2 0.2 0.4>
    gravityOverride {z=-980}
    physicsBodyDefinitions [{
      sphereCollisionRadius=0
      boundBone {boneName=dyn_shirt_front_mid_r}
      boxExtents=<30 30 30>
      constraintSetup {angularLimitsMax=<5 5 5>, angularLimitsMin=<-5 -5 -5>, angularTarget=<1 0 0>}
      localJointOffset=<8 0 0>
     }]
    relativeSpaceBone {boneName=root}
   }
   n14 = Anim Dynamics - Bone: dyn_shirt_front_r(ComponentPose=@n13, Alpha=1, ExternalForce="0,0,0") {
    angularDampingOverride=200
    angularSpringConstant=500
    bGravityOverrideInSimSpace=true
    bUseGravityOverride=true
    simulationSpace=BoneRelative
    alphaBoolBlend {blendOption=Linear}
    boundBone {boneName=dyn_shirt_front_r}
    componentAppliedLinearAccClamp=<1500 1500 1500>
    componentLinearAccScale=<0.4 0.4 1>
    componentLinearVelScale=<0.2 0.2 0.4>
    gravityOverride {z=-980}
    physicsBodyDefinitions [{
      sphereCollisionRadius=0
      boundBone {boneName=dyn_shirt_front_r}
      boxExtents=<30 30 30>
      constraintSetup {angularLimitsMax=<5 5 5>, angularLimitsMin=<-5 -5 -5>, angularTarget=<1 0 0>}
      localJointOffset=<8 0 0>
     }]
    relativeSpaceBone {boneName=root}
   }
   n15 = Anim Dynamics - Bone: dyn_shirt_back_l(ComponentPose=@n14, Alpha=1, ExternalForce="0,0,0") {
    angularDampingOverride=200
    angularSpringConstant=500
    bGravityOverrideInSimSpace=true
    bUseGravityOverride=true
    simulationSpace=BoneRelative
    alphaBoolBlend {blendOption=Linear}
    boundBone {boneName=dyn_shirt_back_l}
    componentAppliedLinearAccClamp=<1500 1500 1500>
    componentLinearAccScale=<0.4 0.4 1>
    componentLinearVelScale=<0.2 0.2 0.4>
    gravityOverride {z=-980}
    physicsBodyDefinitions [{
      sphereCollisionRadius=0
      boundBone {boneName=dyn_shirt_back_l}
      boxExtents=<30 30 30>
      constraintSetup {angularLimitsMax=<5 5 5>, angularLimitsMin=<-5 -5 -5>, angularTarget=<1 0 0>}
      localJointOffset=<8 0 0>
     }]
    relativeSpaceBone {boneName=root}
   }
   n16 = Anim Dynamics - Bone: dyn_shirt_back_mid_l(ComponentPose=@n15, Alpha=1, ExternalForce="0,0,0") {
    angularDampingOverride=200
    angularSpringConstant=500
    bGravityOverrideInSimSpace=true
    bUseGravityOverride=true
    simulationSpace=BoneRelative
    alphaBoolBlend {blendOption=Linear}
    boundBone {boneName=dyn_shirt_back_mid_l}
    componentAppliedLinearAccClamp=<1500 1500 1500>
    componentLinearAccScale=<0.4 0.4 1>
    componentLinearVelScale=<0.2 0.2 0.4>
    gravityOverride {z=-980}
    physicsBodyDefinitions [{
      sphereCollisionRadius=0
      boundBone {boneName=dyn_shirt_back_mid_l}
      boxExtents=<30 30 30>
      constraintSetup {angularLimitsMax=<5 5 5>, angularLimitsMin=<-5 -5 -5>, angularTarget=<1 0 0>}
      localJointOffset=<8 0 0>
     }]
    relativeSpaceBone {boneName=root}
   }
   n17 = Anim Dynamics - Bone: dyn_shirt_front_mid_l(ComponentPose=@n16, Alpha=1, ExternalForce="0,0,0") {
    angularDampingOverride=200
    angularSpringConstant=500
    bGravityOverrideInSimSpace=true
    bUseGravityOverride=true
    simulationSpace=BoneRelative
    alphaBoolBlend {blendOption=Linear}
    boundBone {boneName=dyn_shirt_front_mid_l}
    componentAppliedLinearAccClamp=<1500 1500 1500>
    componentLinearAccScale=<0.4 0.4 1>
    componentLinearVelScale=<0.2 0.2 0.4>
    gravityOverride {z=-980}
    physicsBodyDefinitions [{
      sphereCollisionRadius=0
      boundBone {boneName=dyn_shirt_front_mid_l}
      boxExtents=<30 30 30>
      constraintSetup {angularLimitsMax=<5 5 5>, angularLimitsMin=<-5 -5 -5>, angularTarget=<1 0 0>}
      localJointOffset=<8 0 0>
     }]
    relativeSpaceBone {boneName=root}
   }
   n18 = Anim Dynamics - Bone: dyn_shirt_front_l(ComponentPose=@n17, Alpha=1, ExternalForce="0,0,0") {
    angularDampingOverride=200
    angularSpringConstant=500
    bGravityOverrideInSimSpace=true
    bUseGravityOverride=true
    simulationSpace=BoneRelative
    alphaBoolBlend {blendOption=Linear}
    boundBone {boneName=dyn_shirt_front_l}
    componentAppliedLinearAccClamp=<1500 1500 1500>
    componentLinearAccScale=<0.4 0.4 1>
    componentLinearVelScale=<0.2 0.2 0.4>
    gravityOverride {z=-980}
    physicsBodyDefinitions [{
      sphereCollisionRadius=0
      boundBone {boneName=dyn_shirt_front_l}
      boxExtents=<30 30 30>
      constraintSetup {angularLimitsMax=<5 5 5>, angularLimitsMin=<-5 -5 -5>, angularTarget=<1 0 0>}
      localJointOffset=<8 0 0>
     }]
    relativeSpaceBone {boneName=root}
   }
   n19 = Anim Dynamics - Bone: dyn_collar_r_front(ComponentPose=@n18, Alpha=1, ExternalForce="0,0,0") {
    angularDampingOverride=0.8
    angularSpringConstant=500
    bAngularSpring=true
    bGravityOverrideInSimSpace=true
    bUseGravityOverride=true
    simulationSpace=BoneRelative
    alphaBoolBlend {blendOption=Linear}
    boundBone {boneName=dyn_collar_r_front}
    componentAppliedLinearAccClamp=<1500 1500 1500>
    componentLinearAccScale=<0.4 0.4 0.5>
    componentLinearVelScale=<0.1 0.1 0.2>
    gravityOverride {z=-980}
    physicsBodyDefinitions [{
      sphereCollisionRadius=0
      boundBone {boneName=dyn_collar_r_front}
      boxExtents=<25 25 25>
      constraintSetup {angularLimitsMax=<5 5 2>, angularLimitsMin=<-5 -5 -2>, angularTarget=<1 0 0>}
      localJointOffset=<-1 0 0>
     }]
    relativeSpaceBone {boneName=root}
   }
   n20 = Anim Dynamics - Bone: dyn_collar_r_mid(ComponentPose=@n19, Alpha=1, ExternalForce="0,0,0") {
    angularDampingOverride=0.8
    angularSpringConstant=500
    bAngularSpring=true
    bGravityOverrideInSimSpace=true
    bUseGravityOverride=true
    simulationSpace=BoneRelative
    alphaBoolBlend {blendOption=Linear}
    boundBone {boneName=dyn_collar_r_mid}
    componentAppliedLinearAccClamp=<1500 1500 1500>
    componentLinearAccScale=<0.4 0.4 0.5>
    componentLinearVelScale=<0.1 0.1 0.2>
    gravityOverride {z=-980}
    physicsBodyDefinitions [{
      sphereCollisionRadius=0
      boundBone {boneName=dyn_collar_r_mid}
      boxExtents=<25 25 25>
      constraintSetup {angularLimitsMax=<5 5 2>, angularLimitsMin=<-5 -5 -2>, angularTarget=<1 0 0>}
      localJointOffset=<-1 0 0>
     }]
    relativeSpaceBone {boneName=root}
   }
   n21 = Anim Dynamics - Bone: dyn_collar_r_back(ComponentPose=@n20, Alpha=1, ExternalForce="0,0,0") {
    angularDampingOverride=0.8
    angularSpringConstant=500
    bAngularSpring=true
    bGravityOverrideInSimSpace=true
    bUseGravityOverride=true
    simulationSpace=BoneRelative
    alphaBoolBlend {blendOption=Linear}
    boundBone {boneName=dyn_collar_r_back}
    componentAppliedLinearAccClamp=<1500 1500 1500>
    componentLinearAccScale=<0.4 0.4 0.5>
    componentLinearVelScale=<0.1 0.1 0.2>
    gravityOverride {z=-980}
    physicsBodyDefinitions [{
      sphereCollisionRadius=0
      boundBone {boneName=dyn_collar_r_back}
      boxExtents=<25 25 25>
      constraintSetup {angularLimitsMax=<5 5 2>, angularLimitsMin=<-5 -5 -2>, angularTarget=<1 0 0>}
      localJointOffset=<-1 0 0>
     }]
    relativeSpaceBone {boneName=root}
   }
   n22 = Anim Dynamics - Bone: dyn_collar_l_mid(ComponentPose=@n21, Alpha=1, ExternalForce="0,0,0") {
    angularDampingOverride=0.8
    angularSpringConstant=500
    bAngularSpring=true
    bGravityOverrideInSimSpace=true
    bUseGravityOverride=true
    simulationSpace=BoneRelative
    alphaBoolBlend {blendOption=Linear}
    boundBone {boneName=dyn_collar_l_mid}
    componentAppliedLinearAccClamp=<1500 1500 1500>
    componentLinearAccScale=<0.4 0.4 0.5>
    componentLinearVelScale=<0.1 0.1 0.2>
    gravityOverride {z=-980}
    physicsBodyDefinitions [{
      sphereCollisionRadius=0
      boundBone {boneName=dyn_collar_l_mid}
      boxExtents=<25 25 25>
      constraintSetup {angularLimitsMax=<5 5 2>, angularLimitsMin=<-5 -5 -2>, angularTarget=<1 0 0>}
      localJointOffset=<-1 0 0>
     }]
    relativeSpaceBone {boneName=root}
   }
   n23 = Anim Dynamics - Bone: dyn_collar_l_back(ComponentPose=@n22, Alpha=1, ExternalForce="0,0,0") {
    angularDampingOverride=0.8
    angularSpringConstant=500
    bAngularSpring=true
    bGravityOverrideInSimSpace=true
    bUseGravityOverride=true
    simulationSpace=BoneRelative
    alphaBoolBlend {blendOption=Linear}
    boundBone {boneName=dyn_collar_l_back}
    componentAppliedLinearAccClamp=<1500 1500 1500>
    componentLinearAccScale=<0.4 0.4 0.5>
    componentLinearVelScale=<0.1 0.1 0.2>
    gravityOverride {z=-980}
    physicsBodyDefinitions [{
      sphereCollisionRadius=0
      boundBone {boneName=dyn_collar_l_back}
      boxExtents=<25 25 25>
      constraintSetup {angularLimitsMax=<5 5 2>, angularLimitsMin=<-5 -5 -2>, angularTarget=<1 0 0>}
      localJointOffset=<-1 0 0>
     }]
    relativeSpaceBone {boneName=root}
   }
   n24 = Anim Dynamics - Bone: dyn_collar_back(ComponentPose=@n23, Alpha=1, ExternalForce="0,0,0") {
    angularDampingOverride=0.8
    angularSpringConstant=500
    bAngularSpring=true
    bGravityOverrideInSimSpace=true
    bUseGravityOverride=true
    simulationSpace=BoneRelative
    alphaBoolBlend {blendOption=Linear}
    boundBone {boneName=dyn_collar_back}
    componentAppliedLinearAccClamp=<1500 1500 1500>
    componentLinearAccScale=<0.4 0.4 0.5>
    componentLinearVelScale=<0.1 0.1 0.2>
    gravityOverride {z=-980}
    physicsBodyDefinitions [{
      sphereCollisionRadius=0
      boundBone {boneName=dyn_collar_back}
      boxExtents=<25 25 25>
      constraintSetup {angularLimitsMax=<5 5 2>, angularLimitsMin=<-5 -5 -2>, angularTarget=<1 0 0>}
      localJointOffset=<-1 0 0>
     }]
    relativeSpaceBone {boneName=root}
   }
   n25 = Anim Dynamics - Bone: dyn_collar_l_front(ComponentPose=@n24, Alpha=1, ExternalForce="0,0,0") {
    angularDampingOverride=0.8
    angularSpringConstant=500
    bAngularSpring=true
    bGravityOverrideInSimSpace=true
    bUseGravityOverride=true
    simulationSpace=BoneRelative
    alphaBoolBlend {blendOption=Linear}
    boundBone {boneName=dyn_collar_l_front}
    componentAppliedLinearAccClamp=<1500 1500 1500>
    componentLinearAccScale=<0.4 0.4 0.5>
    componentLinearVelScale=<0.1 0.1 0.2>
    gravityOverride {z=-980}
    physicsBodyDefinitions [{
      sphereCollisionRadius=0
      boundBone {boneName=dyn_collar_l_front}
      boxExtents=<25 25 25>
      constraintSetup {angularLimitsMax=<5 5 2>, angularLimitsMin=<-5 -5 -2>, angularTarget=<1 0 0>}
      localJointOffset=<-1 0 0>
     }]
    relativeSpaceBone {boneName=root}
   }
   n26 = Anim Dynamics - Bone: dyn_collar_buckle_l(ComponentPose=@n25, Alpha=0.8, ExternalForce="0,0,0") {
    angularDampingOverride=0.9
    angularSpringConstant=500
    bGravityOverrideInSimSpace=true
    bOverrideAngularDamping=true
    bUseGravityOverride=true
    simulationSpace=BoneRelative
    alphaBoolBlend {blendOption=Linear}
    boundBone {boneName=dyn_collar_buckle_l}
    componentAppliedLinearAccClamp=<1500 1500 1500>
    componentLinearAccScale=<0.4 0.4 1>
    componentLinearVelScale=<0.2 0.2 0.4>
    gravityOverride {z=-980}
    physicsBodyDefinitions [{
      sphereCollisionRadius=0
      boundBone {boneName=dyn_collar_buckle_l}
      boxExtents=<15 15 15>
      constraintSetup {angularLimitsMax=<3 20 3>, angularLimitsMin=<-3 -20 -3>, angularTarget=<1 0 0>}
      localJointOffset=<5 -5 0>
     }]
    relativeSpaceBone {boneName=root}
   }
   n27 = Anim Dynamics - Bone: dyn_pouch_strap_l(ComponentPose=@n26, Alpha=1, ExternalForce="0,0,0") {
    angularDampingOverride=200
    angularSpringConstant=500
    bGravityOverrideInSimSpace=true
    bUseGravityOverride=true
    simulationSpace=BoneRelative
    alphaBoolBlend {blendOption=Linear}
    boundBone {boneName=dyn_pouch_strap_l}
    componentAppliedLinearAccClamp=<1500 1500 1500>
    componentLinearAccScale=<0.4 0.4 1>
    componentLinearVelScale=<0.2 0.2 0.4>
    gravityOverride {z=-980}
    physicsBodyDefinitions [{
      sphereCollisionRadius=0
      boundBone {boneName=dyn_pouch_strap_l}
      boxExtents=<10 10 10>
      constraintSetup {angularLimitsMax=<0 0 0>, angularLimitsMin=<0 -15 0>, angularTarget=<1 0 0>}
      localJointOffset=<4 0 0>
     }]
    relativeSpaceBone {boneName=root}
   }
   n28 = Anim Dynamics - Bone: dyn_pouch_strap_r(ComponentPose=@n27, Alpha=1, ExternalForce="0,0,0") {
    angularDampingOverride=200
    angularSpringConstant=500
    bGravityOverrideInSimSpace=true
    bUseGravityOverride=true
    simulationSpace=BoneRelative
    alphaBoolBlend {blendOption=Linear}
    boundBone {boneName=dyn_pouch_strap_r}
    componentAppliedLinearAccClamp=<1500 1500 1500>
    componentLinearAccScale=<0.4 0.4 1>
    componentLinearVelScale=<0.2 0.2 0.4>
    gravityOverride {z=-980}
    physicsBodyDefinitions [{
      sphereCollisionRadius=0
      boundBone {boneName=dyn_pouch_strap_r}
      boxExtents=<10 10 10>
      constraintSetup {angularLimitsMax=<0 0 0>, angularLimitsMin=<0 -15 0>, angularTarget=<1 0 0>}
      localJointOffset=<4 0 0>
     }]
    relativeSpaceBone {boneName=root}
   }
   n29 = "Constraint - dyn_shoulder_strap_l_master (PPPPP)"(
    ComponentPose=@n28
    Alpha=1
    ConstraintWeights_0=0.2277
    ConstraintWeights_1=0.1673
    ConstraintWeights_2=0.5161
    ConstraintWeights_3=0.0131
    ConstraintWeights_4=0.0757
   ) {
    alphaBoolBlend {blendOption=Linear}
    boneToModify {boneName=dyn_shoulder_strap_l_master}
    constraintSetup [
     {targetBone {boneName=clavicle_l}}
     {targetBone {boneName=spine_04}}
     {targetBone {boneName=spine_05}}
     {targetBone {boneName=upperarm_twist_01_l}}
     {targetBone {boneName=dyn_shoulder_strap_l}}
    ]
    constraintWeights [1,1,1,1,1]
   }
   n30 = "Constraint - dyn_shoulder_strap_r_master (PPPPP)"!(
    ComponentPose=@n29
    Alpha=1
    ConstraintWeights_0=0.0137
    ConstraintWeights_1:float=0.2211
    ConstraintWeights_2:float=0.097
    ConstraintWeights_3:float=0.1581
    ConstraintWeights_4:float=0.5099
   ) -> (Pose) {
    alphaBoolBlend {blendOption=Linear}
    boneToModify {boneName=dyn_shoulder_strap_r_master}
    constraintSetup [
     {targetBone {boneName=upperarm_twist_01_r}}
     {targetBone {boneName=clavicle_r}}
     {targetBone {boneName=dyn_shoulder_strap_r}}
     {targetBone {boneName=spine_04}}
     {targetBone {boneName=spine_05}}
    ]
    constraintWeights [1,1,1,1,1]
   }
   n31 = Anim Dynamics - Bone: dyn_shoulder_strap_l(ComponentPose=@n30, Alpha=1, ExternalForce="0,0,0") {
    angularDampingOverride=200
    angularSpringConstant=500
    bGravityOverrideInSimSpace=true
    bUseGravityOverride=true
    simulationSpace=BoneRelative
    alphaBoolBlend {blendOption=Linear}
    boundBone {boneName=dyn_shoulder_strap_l}
    componentAppliedLinearAccClamp=<1500 1500 1500>
    componentLinearAccScale=<0.4 0.4 1>
    componentLinearVelScale=<0.2 0.2 0.4>
    gravityOverride {z=-980}
    physicsBodyDefinitions [{
      sphereCollisionRadius=0
      boundBone {boneName=dyn_shoulder_strap_l}
      boxExtents=<10 10 10>
      constraintSetup {angularLimitsMax=<0 0 0>, angularLimitsMin=<0 -15 0>, angularTarget=<1 0 0>}
      localJointOffset=<4 0 0>
     }]
    relativeSpaceBone {boneName=root}
   }
   n32 = Anim Dynamics - Bone: dyn_shoulder_strap_r(ComponentPose=@n31, Alpha=1, ExternalForce="0,0,0") {
    angularDampingOverride=200
    angularSpringConstant=500
    bGravityOverrideInSimSpace=true
    bUseGravityOverride=true
    simulationSpace=BoneRelative
    alphaBoolBlend {blendOption=Linear}
    boundBone {boneName=dyn_shoulder_strap_r}
    componentAppliedLinearAccClamp=<1500 1500 1500>
    componentLinearAccScale=<0.4 0.4 1>
    componentLinearVelScale=<0.2 0.2 0.4>
    gravityOverride {z=-980}
    physicsBodyDefinitions [{
      sphereCollisionRadius=0
      boundBone {boneName=dyn_shoulder_strap_r}
      boxExtents=<10 10 10>
      constraintSetup {angularLimitsMax=<0 0 0>, angularLimitsMin=<0 -15 0>, angularTarget=<1 0 0>}
      localJointOffset=<4 0 0>
     }]
    relativeSpaceBone {boneName=root}
   }
   n33 = "Constraint - dyn_pistol_master (PPPP)"!(
    ComponentPose=@n32
    Alpha=1
    ConstraintWeights_0=0.0077
    ConstraintWeights_1:float=0.7635
    ConstraintWeights_2:float=0
    ConstraintWeights_3:float=0.2288
   ) -> (Pose) {
    alphaBoolBlend {blendOption=Linear}
    boneToModify {boneName=dyn_pistol_master}
    constraintSetup [
     {targetBone {boneName=thigh_r}}
     {targetBone {boneName=pelvis}}
     {targetBone {boneName=thigh_twist_01_l}}
     {targetBone {boneName=thigh_twist_01_r}}
    ]
    constraintWeights [1,1,1,1]
   }
   n34 = "Constraint - dyn_pouch_strap_r_master (PPPPPPPPPPPPPPPPPPPPPPPPPP)"!(
    ComponentPose=@n33
    Alpha=1
    ConstraintWeights_0=0
    ConstraintWeights_1:float=0
    ConstraintWeights_2:float=0
    ConstraintWeights_3:float=0
    ConstraintWeights_4:float=0
    ConstraintWeights_5:float=0
    ConstraintWeights_6:float=0.2163
    ConstraintWeights_7:float=0.0173
    ConstraintWeights_8:float=0
    ConstraintWeights_9:float=0
    ConstraintWeights_10:float=0
    ConstraintWeights_11:float=0.0004
    ConstraintWeights_12:float=0.0001
    ConstraintWeights_13:float=0.1794
    ConstraintWeights_14:float=0
    ConstraintWeights_15:float=0
    ConstraintWeights_16:float=0
    ConstraintWeights_17:float=0.0057
    ConstraintWeights_18:float=0
    ConstraintWeights_19:float=0.0001
    ConstraintWeights_20:float=0
    ConstraintWeights_21:float=0
    ConstraintWeights_22:float=0.5807
    ConstraintWeights_23:float=0
    ConstraintWeights_24:float=0
    ConstraintWeights_25:float=0
   ) -> (Pose) {
    alphaBoolBlend {blendOption=Linear}
    boneToModify {boneName=dyn_pouch_strap_r_master}
    constraintSetup [
     {targetBone {boneName=dyn_shirt_back_mid_l}}
     {targetBone {boneName=dyn_jacket_r_mid}}
     {targetBone {boneName=dyn_shirt_back_l}}
     {targetBone {boneName=spine_01}}
     {targetBone {boneName=spine_02}}
     {targetBone {boneName=dyn_jacket_back}}
     {targetBone {boneName=spine_04}}
     {targetBone {boneName=spine_05}}
     {targetBone {boneName=dyn_shirt_back_mid_r}}
     {targetBone {boneName=dyn_shirt_front_mid_l}}
     {targetBone {boneName=clavicle_l}}
     {targetBone {boneName=upperarm_twist_01_r}}
     {targetBone {boneName=dyn_shirt_front_r}}
     {targetBone {boneName=dyn_jacket_r_front}}
     {targetBone {boneName=dyn_jacket_l_front}}
     {targetBone {boneName=dyn_shirt_front_l}}
     {targetBone {boneName=upperarm_twist_01_l}}
     {targetBone {boneName=clavicle_r}}
     {targetBone {boneName=dyn_jacket_r_back}}
     {targetBone {boneName=dyn_shirt_front_mid_r}}
     {targetBone {boneName=dyn_jacket_l_back}}
     {targetBone {boneName=dyn_shirt_back_r}}
     {targetBone {boneName=dyn_pouch_strap_r}}
     {targetBone {boneName=spine_03}}
     {targetBone {boneName=dyn_pouch_strap_l}}
     {targetBone {boneName=dyn_jacket_l_mid}}
    ]
    constraintWeights [1,1,1,1,1,1,1,1,1,1,1,1,1,1,1,1,1,1,1,1,1,1,1,1,1,1]
   }
   n35 = "Constraint - dyn_pouch_strap_l_master (PPPPPPPPPPPPPPPPPPPPPPPPPP)"!(
    ComponentPose=@n34
    Alpha=1
    ConstraintWeights_0=0
    ConstraintWeights_1:float=0
    ConstraintWeights_2:float=0
    ConstraintWeights_3:float=0
    ConstraintWeights_4:float=0
    ConstraintWeights_5:float=0
    ConstraintWeights_6:float=0.2313
    ConstraintWeights_7:float=0.0146
    ConstraintWeights_8:float=0
    ConstraintWeights_9:float=0
    ConstraintWeights_10:float=0.0053
    ConstraintWeights_11:float=0
    ConstraintWeights_12:float=0
    ConstraintWeights_13:float=0
    ConstraintWeights_14:float=0.1767
    ConstraintWeights_15:float=0
    ConstraintWeights_16:float=0.0001
    ConstraintWeights_17:float=0
    ConstraintWeights_18:float=0
    ConstraintWeights_19:float=0
    ConstraintWeights_20:float=0
    ConstraintWeights_21:float=0
    ConstraintWeights_22:float=0
    ConstraintWeights_23:float=0
    ConstraintWeights_24:float=0.5708
    ConstraintWeights_25:float=0.001
   ) -> (Pose) {
    alphaBoolBlend {blendOption=Linear}
    boneToModify {boneName=dyn_pouch_strap_l_master}
    constraintSetup [
     {targetBone {boneName=dyn_shirt_back_mid_l}}
     {targetBone {boneName=dyn_jacket_r_mid}}
     {targetBone {boneName=dyn_shirt_back_l}}
     {targetBone {boneName=spine_01}}
     {targetBone {boneName=spine_02}}
     {targetBone {boneName=dyn_jacket_back}}
     {targetBone {boneName=spine_04}}
     {targetBone {boneName=spine_05}}
     {targetBone {boneName=dyn_shirt_back_mid_r}}
     {targetBone {boneName=dyn_shirt_front_mid_l}}
     {targetBone {boneName=clavicle_l}}
     {targetBone {boneName=upperarm_twist_01_r}}
     {targetBone {boneName=dyn_shirt_front_r}}
     {targetBone {boneName=dyn_jacket_r_front}}
     {targetBone {boneName=dyn_jacket_l_front}}
     {targetBone {boneName=dyn_shirt_front_l}}
     {targetBone {boneName=upperarm_twist_01_l}}
     {targetBone {boneName=clavicle_r}}
     {targetBone {boneName=dyn_jacket_r_back}}
     {targetBone {boneName=dyn_shirt_front_mid_r}}
     {targetBone {boneName=dyn_jacket_l_back}}
     {targetBone {boneName=dyn_shirt_back_r}}
     {targetBone {boneName=dyn_pouch_strap_r}}
     {targetBone {boneName=spine_03}}
     {targetBone {boneName=dyn_pouch_strap_l}}
     {targetBone {boneName=dyn_jacket_l_mid}}
    ]
    constraintWeights [1,1,1,1,1,1,1,1,1,1,1,1,1,1,1,1,1,1,1,1,1,1,1,1,1,1]
   }
   n36 = RigidBody!(ComponentPose=@n35, Alpha, OverrideWorldGravity:Vector="0,0,-980", ExternalForce) -> (Pose) {
    bClampLinearTranslationLimitToRefPose=true
    bOverrideWorldGravity=true
    overridePhysicsAsset=&/F_MED_Shrapnel_Physics
    simulationSpace=BaseBoneSpace
    alphaBoolBlend {blendOption=Linear}
    baseBoneRef {boneName=root}
    componentAppliedLinearAccClamp=<1500 1500 1500>
    componentLinearAccScale=<0.2 0.2 0.3>
    componentLinearVelScale=<0.2 0.2 0.3>
    overrideWorldGravity {z=-1100}
   }
   n37 = Control Rig(Source=Component To Local(ComponentPose=@n36)) {
    alphaBoolBlend {blendOption=Linear}
    controlRigAssetReference {blueprintRigClass=&/F_MED_Shrapnel_CtrlRig.F_MED_Shrapnel_CtrlRig_C}
   }
   Output Pose(Result=@n37) {layerGroup=None}
  ]}
 EventGraph {nodes [
   PrintString!(execute, InString="To String (Float)"(InDouble=0), bPrintToScreen, bPrintToLog, TextColor, Duration, Key) -> (then)
   n1 = GetDeltaTransformFromRefPose(self=GetOwnerSkeletalMeshComponent(), BoneName=neck_01, BaseName=spine_05)
   MapRangeClamped(Value=BreakRotator(InRot=BreakTransform(InTransform=@n1).Rotation).Yaw, InRangeA=75, InRangeB=105)
   Event BlueprintUpdateAnimation()
  ]}
]
variables []
