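# Physics asset for the F_MED_Shrapnel female cosmetic body part, setting up physics for its side hair and pigtails.
# Left and right hair and pigtail bone chains use small capsule bodies, linked by tight swing and twist limits, with pigtails stiffer than hair.
# Each chain's tiny master body joins the head with free angular motion.
# Head, upper spine and upper arm capsules give body shapes around the chains, also linked by angularly free constraints.
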PhysicsAsset /BRCosmetics/Characters/Player/Female/Medium/Bodies/F_MED_Shrapnel/Meshes/Parts/F_MED_Shrapnel_Physics
bodies {
 dyn_l_hair_01 [{shapeType=Capsule, length=1.25, radius=2.75, center=<2.5672 -0.0001 0>, rotation=<rot 0 0 90>}]
 dyn_l_hair_02 [{shapeType=Capsule, length=1.3395, radius=2.6605, center=<2.5672 -0.0001 0>, rotation=<rot 0 0 90>}]
 dyn_l_hair_03 [{shapeType=Capsule, length=2.5, radius=1.5, center=<2.9153 -0.6889 -0.0001>, rotation=<rot 0.0001 76.7043 89.9998>}]
 dyn_l_hair_master [{shapeType=Capsule, length=0.0001, radius=1.5743}]
 dyn_pigtail_l_01 [{shapeType=Capsule, length=2, radius=2, center=<3.0107 0 0>, rotation=<rot 90 0 0>}]
 dyn_pigtail_l_02 [{shapeType=Capsule, length=2, radius=2, center=<3.0107 0 0>, rotation=<rot 90 0 0>}]
 dyn_pigtail_l_master [{shapeType=Capsule, length=0.0001, radius=1.5743}]
 dyn_pigtail_r_01 [{shapeType=Capsule, length=2, radius=2, center=<3.0107 0 0>, rotation=<rot 90 0 0>}]
 dyn_pigtail_r_02 [{shapeType=Capsule, length=2, radius=2, center=<3.0107 0 0>, rotation=<rot 90 0 0>}]
 dyn_pigtail_r_master [{shapeType=Capsule, length=0.0001, radius=1.5743}]
 dyn_r_hair_01 [{shapeType=Capsule, length=2, radius=2, center=<2.5987 0.0001 0>, rotation=<rot 0 0 90>}]
 dyn_r_hair_02 [{shapeType=Capsule, length=1.5969, radius=2.4031, center=<2.5987 0.0001 0>, rotation=<rot 0 0 90>}]
 dyn_r_hair_03 [{shapeType=Capsule, length=0, radius=2.2748, center=<2.5987 0.0001 0>, rotation=<rot 0 0 90>}]
 dyn_r_hair_master [{shapeType=Capsule, length=0.0001, radius=1.5743}]
 head [{shapeType=Capsule, length=13.75, radius=4.7092, center=<0 3.2524 0>, rotation=<rot 90 90 0>}]
 spine_05 [{shapeType=Capsule, length=17.2686, radius=6.7071, center=<7.6116 -4.388 0>, rotation=<rot 0 -90 0>}]
 upperarm_l [{shapeType=Capsule, length=15.4979, radius=3.2978, center=<12.8027 0.0006 0>, rotation=<rot 90 0 0>}]
 upperarm_r [{shapeType=Capsule, length=15.4979, radius=3.2978, center=<-12.8703 0.0004 0>, rotation=<rot 90 0 0>}]
}
constraints [
 {bone1Name=dyn_pigtail_l_01, bone2Name=dyn_pigtail_l_master, swing1LimitDegrees=15, swing2LimitDegrees=15, twistLimitDegrees=5}
 {bone1Name=dyn_pigtail_l_02, bone2Name=dyn_pigtail_l_01, swing1LimitDegrees=15, swing2LimitDegrees=15, twistLimitDegrees=5}
 {bone1Name=dyn_pigtail_r_01, bone2Name=dyn_pigtail_r_master, swing1LimitDegrees=15, swing2LimitDegrees=15, twistLimitDegrees=5}
 {bone1Name=dyn_pigtail_r_02, bone2Name=dyn_pigtail_r_01, swing1LimitDegrees=15, swing2LimitDegrees=15, twistLimitDegrees=5}
 {bone1Name=dyn_r_hair_01, bone2Name=dyn_r_hair_master, swing1LimitDegrees=30, swing2LimitDegrees=20, twistLimitDegrees=5}
 {bone1Name=dyn_r_hair_02, bone2Name=dyn_r_hair_01, swing1LimitDegrees=30, swing2LimitDegrees=20, twistLimitDegrees=5}
 {bone1Name=dyn_r_hair_03, bone2Name=dyn_r_hair_02, swing1LimitDegrees=30, swing2LimitDegrees=20, twistLimitDegrees=5}
 {bone1Name=dyn_l_hair_01, bone2Name=dyn_l_hair_master, swing1LimitDegrees=30, swing2LimitDegrees=20, twistLimitDegrees=5}
 {bone1Name=dyn_l_hair_02, bone2Name=dyn_l_hair_01, swing1LimitDegrees=30, swing2LimitDegrees=20, twistLimitDegrees=5}
 {bone1Name=dyn_l_hair_03, bone2Name=dyn_l_hair_02, swing1LimitDegrees=30, swing2LimitDegrees=20, twistLimitDegrees=5}
 {
  bone1Name=dyn_pigtail_l_master
  bone2Name=head
  swing1LimitDegrees=45
  swing1Motion=Free
  swing2LimitDegrees=45
  swing2Motion=Free
  twistLimitDegrees=45
  twistMotion=Free
 }
 {
  bone1Name=dyn_pigtail_r_master
  bone2Name=head
  swing1LimitDegrees=45
  swing1Motion=Free
  swing2LimitDegrees=45
  swing2Motion=Free
  twistLimitDegrees=45
  twistMotion=Free
 }
 {
  bone1Name=dyn_r_hair_master
  bone2Name=head
  swing1LimitDegrees=45
  swing1Motion=Free
  swing2LimitDegrees=45
  swing2Motion=Free
  twistLimitDegrees=45
  twistMotion=Free
 }
 {
  bone1Name=dyn_l_hair_master
  bone2Name=head
  swing1LimitDegrees=45
  swing1Motion=Free
  swing2LimitDegrees=45
  swing2Motion=Free
  twistLimitDegrees=45
  twistMotion=Free
 }
 {
  bone1Name=head
  bone2Name=spine_05
  swing1LimitDegrees=45
  swing1Motion=Free
  swing2LimitDegrees=45
  swing2Motion=Free
  twistLimitDegrees=45
  twistMotion=Free
 }
 {
  bone1Name=upperarm_r
  bone2Name=spine_05
  swing1LimitDegrees=45
  swing1Motion=Free
  swing2LimitDegrees=45
  swing2Motion=Free
  twistLimitDegrees=45
  twistMotion=Free
 }
 {
  bone1Name=upperarm_l
  bone2Name=spine_05
  swing1LimitDegrees=45
  swing1Motion=Free
  swing2LimitDegrees=45
  swing2Motion=Free
  twistLimitDegrees=45
  twistMotion=Free
 }
]
